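; Anim notify state for HighTower Date that cuts the playing montage short when the character starts or is firing with either hand.
; Each tick, it reads the left and right firing flags from the linked HighTower_Date_LayerAnimBP layer instance on the mesh.
; If either is set, it stops the notify's montage, including inactive instances, on the player anim blueprint with zero blend-out time.
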(fn Received_NotifyTick
  (params (MeshComp :type Skeletal%20Mesh%20Component%20Object%20Reference) (Animation :type Anim%20Sequence%20Base%20Object%20Reference) (FrameDeltaTime :type Float%20%28single-precision%29) (EventReference :type Anim%20Notify%20Event%20Reference%20Structure%20%28by%20ref%29))
  (bind _asfortnite_m_avg_player_anim_blueprint (Utilities|Casting|CastToFortnite_M_Avg_Player_AnimBlueprint (Components|SkeletalMesh|GetAnimInstance MeshComp)))
  (bind _ashigh_tower_date_layer_anim_bp (Utilities|Casting|CastToHighTower_Date_LayerAnimBP (Components|SkeletalMesh|AnimationBlueprintLinking|GetLinkedAnimLayerInstancebyClass MeshComp "/HighTower/Animation/Date/HighTower_Date_LayerAnimBP.HighTower_Date_LayerAnimBP_C"))
    (:then
      (bind _asanim_montage (Utilities|Casting|CastToAnimMontage Animation)
        (:then
          (if (or (Variables|/Script/HighTowerRuntime.HighTowerDateLayerAnimInstance|GetbIsFiringLeft _ashigh_tower_date_layer_anim_bp) (Variables|/Script/HighTowerRuntime.HighTowerDateLayerAnimInstance|GetbIsFiringRight _ashigh_tower_date_layer_anim_bp))
            (Montage|StopMontageIncludingInactive _asfortnite_m_avg_player_anim_blueprint 0.0 _asanim_montage)
            (else
              _)))
        (:CastFailed
          _)))
    (:CastFailed
      _))
  (return false))
