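; Anim notify for the HighTower Date combo animation that fires the chain lightning weapon at the notify's frame.
; It casts the mesh owner to a FortPawn and acts only if that pawn is locally controlled.
; It then casts the pawn's current weapon to B_HighTower_Date_ChainLightning and calls TrytoCustomFire on it.
; It returns a local Successful flag set after the fire call.
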(fn Received_Notify
  (params (MeshComp :type Skeletal%20Mesh%20Component%20Object%20Reference) (Animation :type Anim%20Sequence%20Base%20Object%20Reference) (EventReference :type Anim%20Notify%20Event%20Reference%20Structure%20%28by%20ref%29))
  (Utilities|FlowControl|Sequence
    (:then_0
      (bind _asfort_pawn (Utilities|Casting|CastToFortPawn (Components|GetOwner MeshComp)))
      (if (Pawn|IsLocallyControlled _asfort_pawn)
        (bind _asb_high_tower_date_chain_lightning (Utilities|Casting|CastToB_HighTower_Date_ChainLightning (Variables|/Script/FortniteGame.FortPawn|GetCurrentWeapon _asfort_pawn)))
        (Class|FortWeapon|TrytoCustomFire _asb_high_tower_date_chain_lightning false)
        (LocalVariables|SetSuccessful true)))
    (:then_1
      (return (LocalVariables|GetSuccessful)))))
