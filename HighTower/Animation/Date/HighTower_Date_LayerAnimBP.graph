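# Linked-layer AnimBP implementing Item Override layers for the HighTower Date item's firing animations on the Fortnite player skeleton.
# ItemUpperBody's state machine alternates Default, FireLeft and FireRight under a yaw AimOffset, choosing pitch blendspaces and bone-masked blends for swimming, crouching, targeting and airborne (right only).
# ItemFinalPoseOverride adds a pitch-driven bomb additive during montages; ItemFallLandAdditive selects between two landing additives.
# Other layers contain no animation logic.
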AnimBlueprint /HighTower/Animation/Date/HighTower_Date_LayerAnimBP
target_skeleton=&Fortnite_M_Avg_Player_Skeleton
graphs [
 AnimGraph {nodes [Output Pose() {layerGroup=None}]}
 EventGraph {links [], nodes []}
 ItemAimOffset {nodes [
   Input Pose!() -> (Pose, AimOffsetAlpha:double, Yaw:double, Pitch:double) {name=InPose_AimOffset}
   Input Pose!() -> (Pose) {name=InPose_PreAimOffsets}
   Output Pose() {layerGroup=Item Override}
  ]}
 ItemFallAdditive {nodes [
   Input Pose() {name=InPoseItemFallAdditive}
   Output Pose() {layerGroup=Item Override}
  ]}
 ItemFallLandAdditive {nodes [
   Input Pose() {name=InPoseItemFallLandAdditive}
   n1 = "Sequence Player (additive) 'HighTower_Date_JumpLand_CMM'"() {bLoopAnimation=false, groupName=JumpLand, method=SyncGroup, sequence=&/HighTower_Date_JumpLand_CMM}
   n2 = "Sequence Player (additive) 'HighTower_Date_RelaxedJumpLand_CMM'"() {bLoopAnimation=false, groupName=JumpLand, method=SyncGroup, sequence=&/HighTower_Date_RelaxedJumpLand_CMM}
   Output Pose(Result=Blend Poses by bool(
     BlendPose_0=@n1
     BlendPose_1=@n2
     BlendTime_0=.25
     BlendTime_1=.25
     bActiveValue=Get bShouldPlayCustomJumpLand() -> (bShouldPlayCustomJumpLand:bool)
    ) {blendType=Linear}) {layerGroup=Item Override}
  ]}
 ItemFallLevel2Override {nodes [
   Input Pose() {name=InPose_FallLevel2}
   Output Pose() {layerGroup=Item Override}
  ]}
 ItemFinalPoseOverride {nodes [
   n1 = Blendspace Player 'HighTowerDate_Combo_Bomb_BS'!(X=Get AimPitch()) -> (Pose) {blendSpace=/HighTower/Animation/Date/Combo/HighTowerDate_Combo_Bomb_BS}
   n2 = Use cached pose 'FinalPose Cache'()
   Output Pose(Result=Blend Poses by bool(
     BlendPose_0=Apply Additive(Base=@n2, Additive=@n1) {alphaBoolBlend {blendOption=Linear}}
     BlendPose_1=Use cached pose 'FinalPose Cache'()
     BlendTime_0=0.3
     BlendTime_1=0.2
     bActiveValue=Get bMontageActive() -> (bMontageActive:bool)
    ) {blendType=Linear}) {layerGroup=Item Override}
   Save cached pose 'FinalPose Cache'(Pose=Input Pose() {name=InFinalPose}) {cacheName=FinalPose Cache}
  ]}
 ItemFullBodyAimOffsets {nodes [
   Input Pose!() -> (Pose, FullBody_AimOffsetAlpha:double, Aim_Yaw:double, Aim_Pitch:double) {name=InPoseFullBodyAimOffsets}
   Input Pose!() -> (Pose) {name=InPosePreFullBodyAimOffsets}
   Output Pose() {layerGroup=Item Override}
  ]}
 ItemFullBodyOverride {nodes [
   Input Pose() {name=InPoseFullBody}
   Output Pose() {layerGroup=Item Override}
  ]}
 ItemIdleOverride {nodes [
   Input Pose() {name=InPose_Idle}
   Output Pose() {layerGroup=Item Override}
  ]}
 ItemJetPackJumpAdditive {nodes [Output Pose() {layerGroup=Item Override}]}
 ItemJetPackStartAdditive {nodes [Output Pose() {layerGroup=Item Override}]}
 ItemJumpLoopAdditive {nodes [
   Input Pose() {name=InPoseItemJumpLoopAdditive}
   Output Pose() {layerGroup=Item Override}
  ]}
 ItemJumpOverride {nodes [
   Input Pose() {name=InPoseJump}
   Output Pose() {layerGroup=Item Override}
  ]}
 ItemJumpUpAdditive {nodes [
   Input Pose() {name=InPoseItemJumpUpAdditive}
   Output Pose() {layerGroup=Item Override}
  ]}
 ItemPreAimOffsets {nodes [
   Input Pose() {name=InPosePreAimOffsets}
   Output Pose() {layerGroup=Item Override}
  ]}
 ItemPreIK {nodes [
   Input Pose() {name=InPose_PreIK}
   Output Pose() {layerGroup=Item Override}
  ]}
 ItemRightHandIK {nodes [
   Input Pose() {name=InPosePreRightHandIK}
   Input Pose() {name=InPosePostRightHandIK}
   Output Pose() {layerGroup=Item Override}
  ]}
 ItemSkeletalControl {nodes [
   Input Pose()
   Output Pose() {layerGroup=Item Override}
  ]}
 ItemSwimJumpFallLoop {nodes [Output Pose() {layerGroup=Item Override}]}
 ItemSwimJumpSurfaceEnd {nodes [Output Pose() {layerGroup=Item Override}]}
 ItemSwimJumpSurfaceLoop {nodes [Output Pose() {layerGroup=Item Override}]}
 ItemUpperBody {nodes [
   AnimGraphNode_StateMachine_0 = DateFireState() {is=title AnimGraphNode_StateMachine}
   n1 = Input Pose!() -> (Pose, InputParam:Fort Anim Input Adjusted Aim Structure) {name=InPoseUpperBody}
   Output Pose(Result=AimOffset Player 'HighTower_Date_YawAO'!(
     X:float=Break Fort Anim Input Adjusted Aim!(FortAnimInput_AdjustedAim:"Fort Anim Input Adjusted Aim Structure (by ref)"=@n1.InputParam) -> (ResultingYaw:float=~, ResultingPitch:float=~, ZeroOutPitchWeightCurveName:Name) {is=title K2Node_BreakStruct}
     Y:float=0
     BasePose:Pose Link Structure=@AnimGraphNode_StateMachine_0
     Alpha:float=1
    ) -> (Pose:Pose Link Structure) {blendSpace=/HighTower/Animation/Date/AO/HighTower_Date_YawAO, alphaBoolBlend {blendOption=Linear}}) {layerGroup=Item Override}
   Save cached pose 'UpperBodyCachePose'(Pose=@n1) {cacheName=UpperBodyCachePose, cachePoseName=UpperBodyCachePose}
  ]}
 DateFireState {
  entry=Default
  path=ItemUpperBody.AnimGraphNode_StateMachine_0.DateFireState
  ids=id
  nodes [
   AnimStateNode_0 = Default() {is=title AnimStateNode}
   AnimStateNode_1 = FireLeft() {is=title AnimStateNode}
   AnimStateNode_2 = FireRight() {is=title AnimStateNode}
   AnimStateTransitionNode_0 = AnimStateTransitionNode() {crossfadeDuration=0.1}
   AnimStateTransitionNode_2 = AnimStateTransitionNode() {crossfadeDuration=0.1}
   AnimStateTransitionNode_3 = AnimStateTransitionNode()
   AnimStateTransitionNode_4 = AnimStateTransitionNode()
   AnimStateTransitionNode_5 = AnimStateTransitionNode() {bAutomaticRuleBasedOnSequencePlayerInState=true, crossfadeDuration=0.25}
   AnimStateTransitionNode_6 = AnimStateTransitionNode() {bAutomaticRuleBasedOnSequencePlayerInState=true, crossfadeDuration=0.25}
  ]
  links [
   {
    dst=FireLeft
    rule=ItemUpperBody.AnimGraphNode_StateMachine_0.DateFireState.AnimStateTransitionNode_0.Transition
    src=Default
    via=AnimStateTransitionNode_0
   }
   {
    dst=FireRight
    rule=ItemUpperBody.AnimGraphNode_StateMachine_0.DateFireState.AnimStateTransitionNode_2.Transition
    src=Default
    via=AnimStateTransitionNode_2
   }
   {
    dst=Default
    rule=ItemUpperBody.AnimGraphNode_StateMachine_0.DateFireState.AnimStateTransitionNode_6.Transition
    src=FireLeft
    via=AnimStateTransitionNode_6
   }
   {
    dst=FireRight
    rule=ItemUpperBody.AnimGraphNode_StateMachine_0.DateFireState.AnimStateTransitionNode_4.Transition
    src=FireLeft
    via=AnimStateTransitionNode_4
   }
   {
    dst=Default
    rule=ItemUpperBody.AnimGraphNode_StateMachine_0.DateFireState.AnimStateTransitionNode_5.Transition
    src=FireRight
    via=AnimStateTransitionNode_5
   }
   {
    dst=FireLeft
    rule=ItemUpperBody.AnimGraphNode_StateMachine_0.DateFireState.AnimStateTransitionNode_3.Transition
    src=FireRight
    via=AnimStateTransitionNode_3
   }
  ]
 }
 Default {
  path=ItemUpperBody.AnimGraphNode_StateMachine_0.DateFireState.AnimStateNode_0.Default
  nodes [Output Animation Pose(Result=Use cached pose 'UpperBodyCachePose'())]
 }
 FireLeft {
  path=ItemUpperBody.AnimGraphNode_StateMachine_0.DateFireState.AnimStateNode_1.FireLeft
  nodes [
   n1 = Get AimPitch()
   n2 = Blendspace Player 'HighTowerDate_Combo_Fire_Left_BS'!(X=@n1) -> (Pose) {bLoop=false, blendSpace=/HighTower/Animation/Date/Combo/HighTowerDate_Combo_Fire_Left_BS, groupName=FireLeft, method=SyncGroup}
   n3 = Blend Poses by bool(
    BlendPose_0=Blendspace Player 'HighTower_Date_FireADS_Left_BS'!(X=@n1) -> (Pose) {blendSpace=&/HighTower_Date_FireADS_Left_BS, groupName=FireLeft, method=SyncGroup}
    BlendPose_1=@n2
    BlendTime_1=0.25
    bActiveValue=Get bIsTargeting()
   ) {blendType=Linear}
   n4 = Blendspace Player 'HighTowerDate_Combo_Fire_Left_BS'(X=@n1) {bLoop=false, blendSpace=&HighTowerDate_Combo_Fire_Left_BS, groupName=FireLeft, method=SyncGroup}
   n5 = Layered blend per bone(BasePose=@n4, BlendPoses_0=Use cached pose 'UpperBodyCachePose'()) {
    bMeshSpaceRotationBlend=true
    blendMode=BlendMask
    blendMasks [&Fortnite_M_Avg_Player_Skeleton:LowerBodyPartialUpperBody]
    layerSetup []
   }
   n6 = Get bIsCrouching()
   n7 = Blendspace Player 'HighTowerDate_Combo_Fire_Left_BS'(X=Get AimPitch()) {bLoop=false, blendSpace=&HighTowerDate_Combo_Fire_Left_BS, groupName=FireLeft, method=SyncGroup}
   n8 = Layered blend per bone(BasePose=Use cached pose 'UpperBodyCachePose'(), BlendPoses_0=@n7) {bMeshSpaceRotationBlend=true, layerSetup [{branchFilters [{boneName=clavicle_l}, {blendDepth=2, boneName=spine_04}]}]}
   Output Animation Pose(Result=Blend Poses by bool(
     BlendPose_0=Modify Curve(SourcePose=@n8, CurveValues_0=1, Alpha=1) {curveNames [DisableArmsHeadAdditive], curveValues [0]}
     BlendPose_1=Blend Poses by bool(BlendPose_0=@n5, BlendPose_1=@n3, bActiveValue=@n6) {blendType=Linear}
     bActiveValue=Get bIsSurfaceSwimming()
    ) {blendType=Linear})
  ]
 }
 FireRight {
  path=ItemUpperBody.AnimGraphNode_StateMachine_0.DateFireState.AnimStateNode_2.FireRight
  nodes [
   n1 = Get AimPitch()
   n2 = Blendspace Player 'HighTowerDate_Combo_Fire_Right_BS'!(X=@n1) -> (Pose) {bLoop=false, blendSpace=/HighTower/Animation/Date/Combo/HighTowerDate_Combo_Fire_Right_BS, groupName=FireRight, method=SyncGroup}
   n3 = Blend Poses by bool(
    BlendPose_0=Blendspace Player 'HighTowerDate_Combo_Fire_Right_InAir_BS'!(X=@n1) -> (Pose) {blendSpace=/HighTower/Animation/Date/Combo/HighTowerDate_Combo_Fire_Right_InAir_BS, groupName=FireRight, method=SyncGroup}
    BlendPose_1=Blend Poses by bool(
     BlendPose_0=Blendspace Player 'HighTower_Date_FireADS_Right_BS'!(X=@n1) -> (Pose) {blendSpace=&/HighTower_Date_FireADS_Right_BS, groupName=FireRight, method=SyncGroup}
     BlendPose_1=@n2
     BlendTime_1=0.25
     bActiveValue=Get bIsTargeting()
    ) {blendType=Linear}
    bActiveValue=NOT Boolean(A=Get bIsOnGround() -> (bIsOnGround:bool))
   ) {blendType=Linear}
   n4 = Blendspace Player 'HighTowerDate_Combo_Fire_Right_BS'(X=@n1) {bLoop=false, blendSpace=&HighTowerDate_Combo_Fire_Right_BS, groupName=FireRight, method=SyncGroup}
   n5 = Layered blend per bone(BasePose=@n4, BlendPoses_0=Use cached pose 'UpperBodyCachePose'()) {
    bMeshSpaceRotationBlend=true
    blendMode=BlendMask
    blendMasks [&Fortnite_M_Avg_Player_Skeleton:LowerBodyPartialUpperBody]
    layerSetup []
   }
   n6 = Layered blend per bone(BasePose=@n5, BlendPoses_0=Use cached pose 'UpperBodyCachePose'(), BlendWeights_0=0.3) {
    bMeshSpaceRotationBlend=true
    blendMode=BlendMask
    blendMasks [&Fortnite_M_Avg_Player_Skeleton:ArmLeft]
    layerSetup []
   }
   n7 = Get bIsCrouching()
   n8 = Blendspace Player 'HighTowerDate_Combo_Fire_Right_BS'(X=Get AimPitch()) {bLoop=false, blendSpace=&HighTowerDate_Combo_Fire_Right_BS, groupName=FireRight, method=SyncGroup}
   n9 = Layered blend per bone(BasePose=Use cached pose 'UpperBodyCachePose'(), BlendPoses_0=@n8) {bMeshSpaceRotationBlend=true, layerSetup [{branchFilters [{boneName=clavicle_r}, {blendDepth=2, boneName=spine_04}]}]}
   Output Animation Pose(Result=Blend Poses by bool(
     BlendPose_0=Modify Curve(SourcePose=@n9, CurveValues_0=1, Alpha=1) {curveNames [DisableArmsHeadAdditive], curveValues [0]}
     BlendPose_1=Blend Poses by bool(BlendPose_0=@n6, BlendPose_1=@n3, bActiveValue=@n7) {blendType=Linear}
     bActiveValue=Get bIsSurfaceSwimming()
    ) {blendType=Linear})
  ]
 }
 Transition {
  path=ItemUpperBody.AnimGraphNode_StateMachine_0.DateFireState.AnimStateTransitionNode_0.Transition
  nodes [Result(bCanEnterTransition=Get bIsFiringLeft() -> (bIsFiringLeft:bool))]
 }
 Transition {
  path=ItemUpperBody.AnimGraphNode_StateMachine_0.DateFireState.AnimStateTransitionNode_2.Transition
  nodes [Result(bCanEnterTransition=Get bIsFiringRight() -> (bIsFiringRight:bool))]
 }
 Transition {
  path=ItemUpperBody.AnimGraphNode_StateMachine_0.DateFireState.AnimStateTransitionNode_3.Transition
  nodes [Result(bCanEnterTransition=Get bIsFiringLeft())]
 }
 Transition {
  path=ItemUpperBody.AnimGraphNode_StateMachine_0.DateFireState.AnimStateTransitionNode_4.Transition
  nodes [Result(bCanEnterTransition=Get bIsFiringRight())]
 }
 Transition {path=ItemUpperBody.AnimGraphNode_StateMachine_0.DateFireState.AnimStateTransitionNode_5.Transition, nodes [Result()]}
 Transition {path=ItemUpperBody.AnimGraphNode_StateMachine_0.DateFireState.AnimStateTransitionNode_6.Transition, nodes [Result()]}
 ItemZipLineStartAdditive {nodes [Output Pose() {layerGroup=Item Override}]}
]
variables []
